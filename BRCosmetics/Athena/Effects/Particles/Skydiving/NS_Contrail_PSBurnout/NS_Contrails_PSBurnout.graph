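# Niagara skydiving contrail cosmetic "PSBurnout", trailing four-colored digital effects from four skeletal mesh bone/socket points.
# Particles partition four ways to pick their attach point, plus color for glows, smoke and bits.
# Pulsing glows track each point while GPU smoke fades toward blue and bits speed up with paraglide lean.
# Ribbons use 40 burst particles, re-placed each update by index along contrail-setup velocity, rather than a spawned trail.
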NiagaraSystem /BRCosmetics/Athena/Effects/Particles/Skydiving/NS_Contrail_PSBurnout/NS_Contrails_PSBurnout
dependencies {
 usedDataInterfaces [
  &NiagaraDataInterfaceCurve
  &NiagaraDataInterfaceArrayColor
  &NiagaraDataInterfaceSpriteRendererInfo
  &NiagaraDataInterfaceColorCurve
  &NiagaraDataInterfaceVector2DCurve
 ]
 usedDynamicInputs []
 usedModules [
  &SystemState
  /Game/Effects/Niagara/Modules/Contrails/FN_Contrail_Setup
  &FN_Integrator
 ]
}
emitters {
 DigitalBits {
  simTarget=GPUComputeSim
  emitterUpdateScript [
   &EmitterState(Inactive Response=::NewEnumerator1, Life Cycle Mode=::NewEnumerator1, Loop Delay=0.2, Loop Duration=1, UseLoopDelay=true)
   &SpawnRate(
    SpawnRate=&RandomRangeFloat(Minimum=12.5, Maximum=50, Evaluation Type=::NewEnumerator1)
    Use Spawn Probability=false
   )
  ]
  particleSpawnScript [
   /SetVariables(Particles.SpriteAlignment=<0 0 1>) {isSetParameters=true}
   &PartitionParticles(Number Of Partitions=4)
   &InitializeParticle(
    Color=&SelectLinearColorFromArray(
     Array Sampling Mode=::NewEnumerator1
     Direct Array Index=$Particles.PartitionParticles.Partition
     Color Selection Array=DataInterface() {propertyValues {ColorData [<rgba 0 10 9.3218 1>, <rgba 10 0 0 1>, <rgba 2 0 1.6187 1>, <rgba 0 0.5677 2 1>]}}
     Random Seed=?
    )
    Lifetime=&RandomRangeFloat(Minimum=0.5, Maximum=1)
    Sprite Size=&RandomRangeVector2D(Minimum=<2 2>, Maximum=<5 5>)
    Sprite Size Mode=::NewEnumerator4
    UsePositionOffset=false
    Write Material Random=true
   )
   &SkeletalMeshLocation(
    Bone / Socket Index=$Particles.PartitionParticles.Partition
    Bone Sampling Mode=::NewEnumerator4
    MaskSpawnGroup=false
    Skeletal Mesh=$User.Skeletal Mesh
   )
   &ShapeLocation(MaskSpawnGroup=false, Override Seed=false, Sphere Radius=10, UseOwnerScale=true)
   /Game/Effects/Niagara/Modules/Skydiving/FN_Skydive_Velocity_v4/FN_Skydive_Velocity(FE Speed=50, Speed:NiagaraFloat=&Lerp_Float(A=150, B=300, Alpha=$User.ParaGlideLeanAlpha))
   &InheritVelocity(Speed Limit=true, UseSpeedThreshold=false)
   &DynamicMaterialParameters(Module.Param0WriteEnabled=true, Module.Param1WriteEnabled=true, Module.Param2WriteEnabled=true, Module.Param3WriteEnabled=true)
  ]
  particleUpdateScript [
   &ParticleState()
   &ScaleSpriteSize(
    Non-Uniform Curve Sprite Scale=DataInterface() {propertyValues {
      XCurve {keys [{tangentMode=RCTM_User, time=0, value=0}, <at 0.5 1>, <at 0.75 0.25>, {tangentMode=RCTM_User, time=1, value=0}]}
      YCurve {keys [{tangentMode=RCTM_User, time=0, value=0}, <at 0.5 1>, <at 0.75 1>, {tangentMode=RCTM_User, time=1, value=20}]}
     }}
    Scale Sprite Size Mode=::NewEnumerator3
   )
   &ScaleColor(
    Scale Alpha=&FloatFromCurve(FloatCurve=DataInterface() {propertyValues {Curve {keys [<at 0.75 1>,<at 1 0>]}}})
    ScaleA=true
    ScaleRGB=false
   )
   &InheritSourceMovement(Applied Position Scale=<0.95 0.95 1>, ApplyForce=false, ApplyPosition=true, ApplyPositionFalloff=false)
   &SolveForcesAndVelocity(Clamp Velocity=false, Limit Acceleration=false)
  ]
  rendererClasses [&NiagaraSpriteRendererProperties]
  renderers [{rendererClass=&NiagaraSpriteRendererProperties}]
 }
 DigitalSmoke {
  simTarget=GPUComputeSim
  emitterUpdateScript [
   &EmitterState(Inactive Response=::NewEnumerator1, Life Cycle Mode=::NewEnumerator1, Loop Duration=1, UseLoopDelay=false)
   &SpawnRate(SpawnRate=112, Use Spawn Probability=false)
  ]
  particleSpawnScript [
   /SetVariables(Particles.SpriteRotationRate=&RandomRangeFloat(Minimum=-0.2, Maximum=0.2)) {isSetParameters=true}
   &PartitionParticles(Number Of Partitions=4)
   &InitializeParticle(
    Color=<rgba 1 1 1 0.7>
    Lifetime=&RandomRangeFloat(Minimum=0.5, Maximum=0.7)
    Sprite Rotation Angle=&RandomRangeFloat(Minimum=0, Maximum=1)
    Sprite Rotation Mode=::NewEnumerator1
    Sprite Size Mode=::NewEnumerator3
    Uniform Sprite Size=32
    UsePositionOffset=false
    Write Material Random=true
   )
   &FN_Skydive_Velocity(FE Speed=100, Speed=0)
   &InheritVelocity(
    Inherited Velocity Amount Scale=&RandomRangeVector(Minimum=<0.8 0.8 0.8>, Maximum=<1 1 1>)
    Inherited Velocity Speed Limit=0
    Speed Limit=false
    UseSpeedThreshold=false
   )
   &SkeletalMeshLocation(
    Bone / Socket Index=$Particles.PartitionParticles.Partition
    Bone Sampling Mode=::NewEnumerator4
    MaskSpawnGroup=false
    Skeletal Mesh=$User.Skeletal Mesh
   )
  ]
  particleUpdateScript [
   &ParticleState()
   &SpriteRotationRate(Rotation Rate=&AngleConversion(
     Angle:NiagaraFloat=$Particles.SpriteRotationRate
     Angle Input:ENiagara_AngleInput=::NewEnumerator1
     Angle Output:ENiagara_AngleInput=::NewEnumerator0
    ))
   &ScaleSpriteSize(Uniform Curve Sprite Scale=DataInterface() {propertyValues {Curve {keys [{tangentMode=RCTM_User, time=0, value=0.6}, {tangentMode=RCTM_User, time=1, value=2.6}]}}})
   &CameraOffset(Camera Offset Amount=-25)
   &SubUVAnimation(
    End Frame Range Override=3
    "Sprite Renderer "=DataInterface() {propertyValues {SpriteRenderer=&self:DigitalSmoke_2.NiagaraSpriteRendererProperties_1}}
    UseEndFrame=true
    UseStartFrame=true
   )
   &ScaleColor(
    Scale Alpha=&FloatFromCurve(FloatCurve=DataInterface() {propertyValues {Curve {keys [
         <at 0 0 RCIM_Cubic>
         {arriveTangent=1.6418, interpMode=RCIM_Cubic, leaveTangent=1.6418, time=0.1085, value=0.7977}
         {arriveTangent=-1.2609, interpMode=RCIM_Cubic, leaveTangent=-1.2609, time=0.3683, value=0.6047}
         {arriveTangent=-0.9574, interpMode=RCIM_Cubic, leaveTangent=-0.9574, time=0.6265, value=0.1445}
         <at 1 0>
        ]}}})
    Scale RGB=&MakeVectorFromLinearColorRGB(Color=&LerpLinearColors(
      StartColor=&SelectLinearColorFromArray(
       Array Sampling Mode=::NewEnumerator1
       Direct Array Index=$Particles.PartitionParticles.Partition
       Color Selection Array=DataInterface() {propertyValues {ColorData [<rgba 0 2 1.8644 1>, <rgba 3 0 0 1>, <rgba 2 0 1.6187 1>, <rgba 0 0.5677 2 1>]}}
       Random Seed=?
      )
      EndColor=<rgba 0 0.1349 1 1>
      LerpFactor=&FloatFromCurve(FloatCurve=DataInterface() {propertyValues {Curve {keys [<at 0 0>,<at 0.4 1>]}}})
     ))
    ScaleA=true
    ScaleRGB=true
   )
   &DynamicMaterialParameters(
    Index 0 Param 1=&FloatFromCurve(FloatCurve=DataInterface() {propertyValues {Curve {keys [<at 0 0.8272 RCIM_Cubic>,<at 1 1.146 RCIM_Cubic>]}}})
    Index 0 Param 2=&FloatFromCurve(FloatCurve=DataInterface() {propertyValues {Curve {keys [<at 0 0.05>,<at 1 0.05>]}}})
    Index 0 Param 3=&FloatFromCurve(FloatCurve=DataInterface() {propertyValues {Curve {keys [<at 0 0.4329 RCIM_Cubic>,<at 0.3871 0.0669 RCIM_Cubic>]}}})
    Module.Param0WriteEnabled=true
    Module.Param1WriteEnabled=true
    Module.Param2WriteEnabled=true
    Module.Param3WriteEnabled=true
   )
   &SolveForcesAndVelocity(Clamp Velocity=false, Limit Acceleration=false)
  ]
  rendererClasses [&NiagaraSpriteRendererProperties]
  renderers [{rendererClass=&NiagaraSpriteRendererProperties}]
 }
 Glows {
  emitterUpdateScript [
   &EmitterState(
    Inactive Response=::NewEnumerator1
    Life Cycle Mode=::NewEnumerator1
    Loop Behavior=::NewEnumerator1
    Loop Duration Mode=::NewEnumerator1
    UseLoopDelay=false
   )
   &SpawnBurst_Instantaneous(Spawn Count=4, Use Spawn Probability=false, UseLoopCountLimit=false)
  ]
  particleSpawnScript [
   &PartitionParticles(Number Of Partitions=4)
   &InitializeParticle(
    Color=&SelectLinearColorFromArray(
     Array Sampling Mode=::NewEnumerator1
     Direct Array Index=$Particles.PartitionParticles.Partition
     Color Selection Array=DataInterface() {propertyValues {ColorData [<rgba 0 1 0.9322 1>, <rgba 5 0 0 1>, <rgba 1 0 0.8094 1>, <rgba 0 0.2838 1 1>]}}
     Random Seed=?
    )
    Sprite Size Mode=::NewEnumerator3
    Uniform Sprite Size=90
    UsePositionOffset=false
    Write Material Random=true
   )
  ]
  particleUpdateScript [
   &ParticleState(Kill Particles When Lifetime Has Elapsed=false, Let Infinitely Lived Particles Die When Emitter Deactivates=true)
   &SkeletalMeshLocation(
    Bone / Socket Index=$Particles.PartitionParticles.Partition
    Bone Sampling Mode=::NewEnumerator4
    MaskSpawnGroup=false
    Skeletal Mesh=$User.Skeletal Mesh
   )
   &CameraOffset(Camera Offset Amount=40)
   &ScaleColor(
    Scale Alpha=&Waveform(
     Global Input Position=&ReturnNormalizedExecIndex(Normalized Index Scale=20)
     Global Phase=&Add_Float(A=$Emitter.Age, B=&RandomRangeFloat(Minimum=0, Maximum=5))
     "Waveform [1]"=::NewEnumerator0
     "[1] Frequency"=1
     "[1] Amplitude Min/Max"=<0.25 0.35>
    )
    ScaleA=true
    ScaleRGB=false
   )
   /SetVariables(Particles.SpriteSizePulse:Vector2f=&Vector2DFromFloat(Value=&Waveform(
      Global Input Position=&ReturnNormalizedExecIndex(Normalized Index Scale=20)
      Global Phase=&Add_Float(A=$Emitter.Age, B=&RandomRangeFloat(Minimum=0, Maximum=5))
      "Waveform [1]"=::NewEnumerator0
      "[1] Frequency"=1
      "[1] Amplitude Min/Max"=<65 80>
     ))) {isSetParameters=true}
  ]
  rendererClasses [&NiagaraSpriteRendererProperties]
  renderers [{rendererClass=&NiagaraSpriteRendererProperties}]
 }
 Playstation_Ribbons {
  emitterUpdateScript [
   &EmitterState()
   &SpawnBurst_Instantaneous(Spawn Count=40, Use Spawn Probability=false, UseLoopCountLimit=true)
  ]
  particleSpawnScript [
   &PartitionParticles(
    Normalized Index In Partition Includes One=true
    Number Of Partitions=4
    Particles Per Partition=10
    Partition Mode:ENiagara_PartitionMode=::NewEnumerator1
    Write Partition Index to Ribbon Link Order:ENiagara_AttributeSamplingApplyOutput=::NewEnumerator0
    Write Partition to Ribbon ID=::NewEnumerator0
   )
   &InitializeParticle(
    Color Mode=::NewEnumerator0
    Lifetime=0.6
    Ribbon Width=15
    Ribbon Width Mode=::NewEnumerator1
    Sprite Size Mode=::NewEnumerator3
    UsePositionOffset=false
    Write Material Random=true
   )
  ]
  particleUpdateScript [
   &ParticleState(Kill Particles When Lifetime Has Elapsed=false)
   &SkeletalMeshLocation(
    Bone / Socket Index=$Particles.PartitionParticles.Partition
    Bone Sampling Mode=::NewEnumerator4
    MaskSpawnGroup=false
    Skeletal Mesh=$User.Skeletal Mesh
   )
   /SetVariables(Particles.NewAge:NiagaraFloat=&Multiply_Float(A=$Particles.PartitionParticles.Normalized Index In Partition, B=$Particles.Initial.Lifetime)) {isSetParameters=true}
   /SetVariables(Particles.Velocity=&Multiply_VectorByFloat(
     Vector=&Lerp_Vector(A=$System.FN_Contrail_Setup.Direction Velocity Raw, B=$System.FN_Contrail_Setup.Direction Velocity, Alpha=$Particles.NewAge)
     Float=$Particles.Initial.Lifetime
    )) {isSetParameters=true}
   /SetVariables(Particles.Position=&AddVectorToPosition(
     Position=$Particles.Position
     Vector=&Multiply_VectorByFloat(Vector=$Particles.Velocity, Float=$Particles.PartitionParticles.Normalized Index In Partition)
    )) {isSetParameters=true}
   &ScaleColor(Scale Alpha=$System.FN_Contrail_Setup.Normalized Intro Age, ScaleA=true, ScaleRGB=true)
   &DynamicMaterialParameters(
    Index 0 Param 2=1
    Index 0 Param 3=-5
    Index 0 Param 4=-5
    Module.Param0WriteEnabled=true
    Module.Param1WriteEnabled=true
    Module.Param2WriteEnabled=true
    Module.Param3WriteEnabled=true
   )
   &ScaleColor(
    Scale Alpha=&FloatFromCurve(
     FloatCurve=DataInterface() {propertyValues {Curve {keys [
         <at 0 0 RCIM_Cubic>
         {arriveTangent=-0.0384, interpMode=RCIM_Cubic, leaveTangent=-0.0384, tangentMode=RCTM_User, time=0.25, value=1}
         {arriveTangent=0.048, interpMode=RCIM_Cubic, leaveTangent=0.048, tangentMode=RCTM_User, time=0.5, value=1}
         <at 1 0 RCIM_Cubic>
        ]}}}
     CurveIndex=$Particles.PartitionParticles.Normalized Index In Partition
    )
    ScaleA=true
    ScaleRGB=false
   ) {module=ScaleColor001}
  ]
  rendererClasses [&NiagaraSpriteRendererProperties, &NiagaraRibbonRendererProperties]
  renderers [
   {rendererClass=&NiagaraSpriteRendererProperties}
   {rendererClass=&NiagaraRibbonRendererProperties, rendererIndex=1}
  ]
 }
}
summary {
 systemName=NS_Contrails_PSBurnout
 emitters [
  {
   emitterName=Playstation_Ribbons
   rendererClasses [&NiagaraSpriteRendererProperties, &NiagaraRibbonRendererProperties]
  }
  {emitterName=Glows, rendererClasses [&NiagaraSpriteRendererProperties]}
  {emitterName=DigitalSmoke, simTarget=GPUComputeSim, rendererClasses [&NiagaraSpriteRendererProperties]}
  {emitterName=DigitalBits, simTarget=GPUComputeSim, rendererClasses [&NiagaraSpriteRendererProperties]}
 ]
 userVariables [
  User.bIsFrontEnd {type=NiagaraBool}
  User.bIsFrontEndPreview {type=NiagaraBool, defaultValue=true}
  User.ParaGlideLeanAlpha {type=NiagaraFloat, defaultValue=0, description=""}
  User.RibbonIDOffset {type=NiagaraInt32}
  User.Skeletal Mesh {
   type=NiagaraDataInterfaceSkeletalMesh
   defaultValue {
    struct=NiagaraExt_VariableValue_DataInterface
    value {dataInterface=&self:NiagaraDataInterfaceSkeletalMesh_2}
   }
  }
  User.SkeletalMesh {
   type=NiagaraDataInterfaceSkeletalMesh
   defaultValue {
    struct=NiagaraExt_VariableValue_DataInterface
    value {dataInterface=&self:NiagaraDataInterfaceSkeletalMesh_3}
   }
  }
 ]
}
